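# Post-process style AnimBlueprint for the StrontiumSpark male character body, adding secondary motion to jacket, emblem and chain.
# It copies the parent pose, layers in head-mesh facial bones, then runs Control Rig, constraints, AnimDynamics, leg look-ats and RigidBody simulation.
# EventGraph updates gravity and alphas from thigh angles, forward speed, melee and skydive floating, driving corrective bone offsets and pelvis constraints.
# When rigid bodies are disabled, output is the plain copied parent pose.
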AnimBlueprint /Game/Characters/Player/Male/Medium/Bodies/M_MED_StrontiumSpark/Meshes/M_MED_StrontiumSpark_AnimBP
target_skeleton=&/M_MED_StrontiumSpark_Skeleton
graphs [
 AnimGraph {nodes [
   "Transform (Modify) Bone - Bone: jacket_front_base_L"(
    Alpha=Get LeftLegForwardBendCompensationAlpha() -> (LeftLegForwardBendCompensationAlpha:double)
    Translation="0,0,0"
    Rotation="-10,0,-10"
    Scale="1,1,1"
   ) {
    rotationMode=BMM_Additive
    rotationSpace=BCS_BoneSpace
    translationMode=BMM_Additive
    translationSpace=BCS_BoneSpace
    alphaBoolBlend {blendOption=Linear}
    alphaScaleBiasClamp {bInterpResult=true}
    boneToModify {boneName=jacket_front_base_L}
    rotation {pitch=-10, roll=-10}
   }
   n1 = Get FloatConstraint_alpha() -> (FloatConstraint_alpha:double)
   n2 = Layered blend per bone(
    BasePose=Copy Pose From Mesh() {bUseAttachedParent=true}
    BlendPoses_0=Copy Pose From Mesh(SourceMeshComponent=Get head_mesh() -> (head_mesh:Skeletal Mesh Component Object Reference))
   ) {layerSetup [{branchFilters [
       {boneName=FACIAL_C_UnderChin}
       {boneName=FACIAL_L_UnderChin}
       {boneName=FACIAL_R_UnderChin}
       {boneName=FACIAL_R_Jawline}
       {boneName=FACIAL_C_Jawline}
       {boneName=FACIAL_L_Jawline}
       {boneName=FACIAL_C_FacialRoot}
       {boneName=FACIAL_C_Jaw}
       {boneName=FACIAL_C_Chin}
       {boneName=FACIAL_L_ChinSide}
       {boneName=FACIAL_R_ChinSide}
       {boneName=FACIAL_L_CheekLower}
       {boneName=FACIAL_L_Ear}
       {boneName=FACIAL_L_Masseter}
       {boneName=FACIAL_R_Ear}
       {boneName=FACIAL_R_CheekLower}
       {boneName=FACIAL_R_Masseter}
      ]}]}
   n3 = Control Rig(Source=@n2) {
    alphaBoolBlend {blendOption=Linear}
    controlRigAssetReference {blueprintRigClass=&/M_MED_StrontiumSpark_CtrlRig.M_MED_StrontiumSpark_CtrlRig_C}
   }
   n4 = "Constraint - dyn_jacketemblem_mstr (PPPPPPPP)"!(
    ComponentPose=Local To Component(LocalPose=@n3)
    Alpha=1
    ConstraintWeights_0=0.1444
    ConstraintWeights_1:float=0.4887
    ConstraintWeights_2:float=0.2454
    ConstraintWeights_3:float=0.0716
    ConstraintWeights_4:float=0.0066
    ConstraintWeights_5:float=0.028
    ConstraintWeights_6:float=0.0063
    ConstraintWeights_7:float=0.0091
   ) -> (Pose) {
    alphaBoolBlend {blendOption=Linear}
    boneToModify {boneName=dyn_jacketemblem_mstr}
    constraintSetup [
     {targetBone {boneName=spine_04}}
     {targetBone {boneName=spine_05}}
     {targetBone {boneName=clavicle_l}}
     {targetBone {boneName=spine_03}}
     {targetBone {boneName=clavicle_r}}
     {targetBone {boneName=neck_01}}
     {targetBone {boneName=deform_upperarm_up_l}}
     {targetBone {boneName=deform_upperarm_fr_l}}
    ]
    constraintWeights [1,1,1,1,1,1,1,1]
   }
   n5 = Anim Dynamics - Bone: dyn_jacketchain_01(ComponentPose=@n4, Alpha=1, ExternalForce="0,0,0") {
    angularDampingOverride=0.9999
    angularSpringConstant=50
    bAngularSpring=true
    bOverrideAngularDamping=true
    bUseGravityOverride=true
    simulationSpace=RootRelative
    alphaBoolBlend {blendOption=Linear}
    boundBone {boneName=dyn_jacketchain_01}
    componentAppliedLinearAccClamp=<3000 3000 3000>
    componentLinearAccScale=<0.3 0.3 0.3>
    componentLinearVelScale=<0.3 0.3 0.3>
    gravityOverride {z=-980}
    physicsBodyDefinitions [{
      boundBone {boneName=dyn_jacketchain_01}
      boxExtents=<5 10 10>
      constraintSetup {angularLimitsMax=<10 0 8>, angularLimitsMin=<-10 -80 -10>, angularTarget=<0 0 0>}
      localJointOffset=<1 0 0>
     }]
    relativeSpaceBone {boneName=pelvis}
   }
   n6 = Anim Dynamics - Bone: dyn_jacketchain_02(ComponentPose=@n5, Alpha=1, ExternalForce="0,0,0") {
    angularDampingOverride=0.9999
    angularSpringConstant=50
    bAngularSpring=true
    bOverrideAngularDamping=true
    bUseGravityOverride=true
    simulationSpace=RootRelative
    alphaBoolBlend {blendOption=Linear}
    boundBone {boneName=dyn_jacketchain_02}
    componentAppliedLinearAccClamp=<3000 3000 3000>
    componentLinearAccScale=<0.3 0.3 0.3>
    componentLinearVelScale=<0.3 0.3 0.3>
    gravityOverride {z=-980}
    physicsBodyDefinitions [{
      boundBone {boneName=dyn_jacketchain_02}
      boxExtents=<5 10 10>
      constraintSetup {angularLimitsMax=<10 0 10>, angularLimitsMin=<-10 -60 -15>, angularTarget=<0 0 0>}
      localJointOffset=<1 0 0>
     }]
    relativeSpaceBone {boneName=pelvis}
   }
   n7 = Anim Dynamics - Bone: dyn_jacketchain_03(ComponentPose=@n6, Alpha=1, ExternalForce="0,0,0") {
    angularDampingOverride=0.9999
    angularSpringConstant=50
    bAngularSpring=true
    bOverrideAngularDamping=true
    bUseGravityOverride=true
    simulationSpace=RootRelative
    alphaBoolBlend {blendOption=Linear}
    boundBone {boneName=dyn_jacketchain_03}
    componentAppliedLinearAccClamp=<3000 3000 3000>
    componentLinearAccScale=<0.3 0.3 0.3>
    componentLinearVelScale=<0.3 0.3 0.3>
    gravityOverride {z=-980}
    physicsBodyDefinitions [{
      boundBone {boneName=dyn_jacketchain_03}
      boxExtents=<5 10 10>
      constraintSetup {angularLimitsMax=<10 0 20>, angularLimitsMin=<-10 -60 -20>, angularTarget=<0 0 0>}
      localJointOffset=<1 0 0>
     }]
    relativeSpaceBone {boneName=pelvis}
   }
   n8 = Look At - Bone: jacket_front_base_L(ComponentPose=@n7, Alpha=0.5) {
    alphaBoolBlend {blendOption=Linear}
    boneToModify {boneName=jacket_front_base_L}
    lookAtLocation=<0 10>
    lookAtTarget {boneReference {boneName=calf_l}}
    lookAt_Axis {axis=<1 0>}
   }
   n9 = Look At - Bone: jacket_base_L(ComponentPose=@n8, Alpha=0.3) {
    alphaBoolBlend {blendOption=Linear}
    boneToModify {boneName=jacket_base_L}
    lookAtLocation=<0 10>
    lookAtTarget {boneReference {boneName=calf_l}}
    lookAt_Axis {axis=<1 0>}
   }
   n10 = Look At - Bone: jacket_front_base_R(ComponentPose=@n9, Alpha=1) {
    alphaBoolBlend {blendOption=Linear}
    boneToModify {boneName=jacket_front_base_R}
    lookAtLocation=<0 -10>
    lookAtTarget {boneReference {boneName=calf_r}}
    lookAt_Axis {axis=<1 0>}
   }
   n11 = Look At - Bone: jacket_base_R(ComponentPose=@n10, Alpha=0.7) {
    alphaBoolBlend {blendOption=Linear}
    boneToModify {boneName=jacket_base_R}
    lookAtLocation=<0 -10>
    lookAtTarget {boneReference {boneName=calf_r}}
    lookAt_Axis {axis=<1 0>}
   }
   n12 = RigidBody!(ComponentPose=@n11, Alpha, OverrideWorldGravity:Vector=Get Gravity Override(), ExternalForce) -> (Pose) {
    bClampLinearTranslationLimitToRefPose=true
    bOverrideWorldGravity=true
    overridePhysicsAsset=&/M_MED_StrontiumSpark_Physics
    simulationSpace=BaseBoneSpace
    alphaBoolBlend {blendOption=Linear}
    alphaScaleBiasClamp {interpSpeedDecreasing=15, interpSpeedIncreasing=20}
    baseBoneRef {boneName=pelvis}
    componentAppliedLinearAccClamp=<3000 3000 3000>
    componentLinearAccScale=<0.35 0.35 0.35>
    componentLinearVelScale=<0.35 0.35 0.35>
    overrideWorldGravity {z=-980}
   }
   n13 = "Transform (Modify) Bone - Bone: dyn_jacket_front_02_L"(ComponentPose=@n12, Alpha=Get ForwardMovementAlpha() -> (ForwardMovementAlpha:double), Translation="0,0,-2", Rotation="0,0,0", Scale="1,1,1") {
    rotationMode=BMM_Additive
    rotationSpace=BCS_BoneSpace
    translationMode=BMM_Additive
    translationSpace=BCS_BoneSpace
    alphaBoolBlend {blendOption=Linear}
    alphaScaleBiasClamp {bInterpResult=true}
    boneToModify {boneName=dyn_jacket_front_02_L}
    translation {z=-2}
   }
   n14 = "Transform (Modify) Bone - Bone: dyn_jacket_front_01_L"(ComponentPose=@n13, Alpha=Get pickaxeSwing() -> (pickaxeSwing:double), Translation="-2,0,-0.5", Rotation="0,0,0", Scale="1,1,1") {
    rotationMode=BMM_Additive
    rotationSpace=BCS_BoneSpace
    translationMode=BMM_Additive
    translationSpace=BCS_BoneSpace
    alphaBoolBlend {blendOption=Linear}
    alphaScaleBiasClamp {bInterpResult=true, interpSpeedDecreasing=7, interpSpeedIncreasing=5}
    boneToModify {boneName=dyn_jacket_front_01_L}
    translation {x=-2, z=-0.5}
   }
   n15 = "Constraint - dyn_jacket_01_L (P)"(ComponentPose=@n14, Alpha=@n1, ConstraintWeights_0=1) {
    alphaBoolBlend {blendOption=Linear}
    boneToModify {boneName=dyn_jacket_01_L}
    constraintSetup [{targetBone {boneName=pelvis}}]
    constraintWeights [1]
   }
   n16 = "Constraint - dyn_jacket_01_R (P)"(ComponentPose=@n15, Alpha=@n1, ConstraintWeights_0=1) {
    alphaBoolBlend {blendOption=Linear}
    boneToModify {boneName=dyn_jacket_01_R}
    constraintSetup [{targetBone {boneName=pelvis}}]
    constraintWeights [1]
   }
   n17 = "Constraint - dyn_jacket_front_01_L (P)"(ComponentPose=@n16, Alpha=@n1, ConstraintWeights_0=1) {
    alphaBoolBlend {blendOption=Linear}
    boneToModify {boneName=dyn_jacket_front_01_L}
    constraintSetup [{targetBone {boneName=pelvis}}]
    constraintWeights [1]
   }
   n18 = "Constraint - dyn_jacket_front_01_R (P)"(ComponentPose=@n17, Alpha=@n1, ConstraintWeights_0=1) {
    alphaBoolBlend {blendOption=Linear}
    boneToModify {boneName=dyn_jacket_front_01_R}
    constraintSetup [{targetBone {boneName=pelvis}}]
    constraintWeights [1]
   }
   n19 = "Constraint - dyn_jacket_back_01_L (P)"(ComponentPose=@n18, Alpha=@n1, ConstraintWeights_0=1) {
    alphaBoolBlend {blendOption=Linear}
    boneToModify {boneName=dyn_jacket_back_01_L}
    constraintSetup [{targetBone {boneName=pelvis}}]
    constraintWeights [1]
   }
   n20 = "Constraint - dyn_jacket_back_01_R (P)"(ComponentPose=@n19, Alpha=@n1, ConstraintWeights_0=1) {
    alphaBoolBlend {blendOption=Linear}
    boneToModify {boneName=dyn_jacket_back_01_R}
    constraintSetup [{targetBone {boneName=pelvis}}]
    constraintWeights [1]
   }
   n21 = "Constraint - dyn_jacket_02_L (P)"(ComponentPose=@n20, Alpha=@n1, ConstraintWeights_0=1) {
    alphaBoolBlend {blendOption=Linear}
    boneToModify {boneName=dyn_jacket_02_L}
    constraintSetup [{targetBone {boneName=pelvis}}]
    constraintWeights [1]
   }
   n22 = "Constraint - dyn_jacket_02_R (P)"(ComponentPose=@n21, Alpha=@n1, ConstraintWeights_0=1) {
    alphaBoolBlend {blendOption=Linear}
    boneToModify {boneName=dyn_jacket_02_R}
    constraintSetup [{targetBone {boneName=pelvis}}]
    constraintWeights [1]
   }
   n23 = "Constraint - dyn_jacket_front_02_L (P)"(ComponentPose=@n22, Alpha=@n1, ConstraintWeights_0=1) {
    alphaBoolBlend {blendOption=Linear}
    boneToModify {boneName=dyn_jacket_front_02_L}
    constraintSetup [{targetBone {boneName=pelvis}}]
    constraintWeights [1]
   }
   n24 = "Constraint - dyn_jacket_front_02_R (P)"(ComponentPose=@n23, Alpha=@n1, ConstraintWeights_0=1) {
    alphaBoolBlend {blendOption=Linear}
    boneToModify {boneName=dyn_jacket_front_02_R}
    constraintSetup [{targetBone {boneName=pelvis}}]
    constraintWeights [1]
   }
   n25 = "Constraint - dyn_jacket_back_02_L (P)"(ComponentPose=@n24, Alpha=@n1, ConstraintWeights_0=1) {
    alphaBoolBlend {blendOption=Linear}
    boneToModify {boneName=dyn_jacket_back_02_L}
    constraintSetup [{targetBone {boneName=pelvis}}]
    constraintWeights [1]
   }
   n26 = "Constraint - dyn_jacket_back_02_R (P)"(ComponentPose=@n25, Alpha=@n1, ConstraintWeights_0=1) {
    alphaBoolBlend {blendOption=Linear}
    boneToModify {boneName=dyn_jacket_back_02_R}
    constraintSetup [{targetBone {boneName=pelvis}}]
    constraintWeights [1]
   }
   n27 = "Transform (Modify) Bone - Bone: jacket_front_base_R"(ComponentPose=@n26, Alpha=Get RThighFrontUpAlpha() -> (RThighFrontUpAlpha:double), Translation="0,0,0", Rotation="-2,0,-5", Scale="1,1,1") {rotationMode=BMM_Additive, alphaBoolBlend {blendOption=Linear}, boneToModify {boneName=jacket_front_base_R}, rotation {pitch=-2, roll=-5}}
   n28 = "Transform (Modify) Bone - Bone: jacket_front_base_L"(ComponentPose=@n27, Alpha=Get LThighFrontUpAlpha() -> (LThighFrontUpAlpha:double), Translation="0,0,0", Rotation="0,0,-7", Scale="1,1,1") {rotationMode=BMM_Additive, alphaBoolBlend {blendOption=Linear}, boneToModify {boneName=jacket_front_base_L}, rotation {roll=-7}}
   Output Pose(Result=Blend Poses by bool(
     BlendPose_0=Component To Local(ComponentPose=@n28)
     BlendPose_1=Copy Pose From Mesh() {bUseAttachedParent=true}
     bActiveValue=Get bRigidBodyEnabled()
    ) {blendType=Linear}) {layerGroup=None}
  ]}
 EventGraph {nodes [
   Event OnCharacterCustomizationLoaded()
   > n1 = Set Skel Mesh(Skel Mesh=GetPartSkeletalMeshComponent())
   > n2 = Get Dynamic Material Instance()
   > Set Mat Instance Dynamic(Mat Instance Dynamic=@n2)
   > n3 = GetOrCreateDynamicOverlayMaterial(MeshComponent=@n1, MaterialName=None)
   > Set Overlay Mat Instance Dynamic(Overlay Mat Instance Dynamic=@n3)
   Event BlueprintInitializeAnimation()
   > Set Skel Mesh(Skel Mesh=GetOwnerSkeletalMeshComponent())
   Event BlueprintUpdateAnimation()
   > n4 = Sequence!(execute) -> (then_0, then_1, then_2:Exec, then_3:Exec, then_4:Exec)
   > n5 = Branch(Condition=Get bIsPlayingMeleeAnim())
   > Set pickaxeSwing(pickaxeSwing:double=1) -> (Output_Get:double)
   Set pickaxeSwing(execute=@n5.else, pickaxeSwing=0)
   Is Valid(exec=@n4.then_3, InputObject=Get Skel Mesh())
   > n6 = Gravity Override State Function DEPRECATED(CharacterPartAnimInstance=Self-Reference(), GameTimeInSeconds=GetGameTimeInSeconds(), GravityOverrideParams=Get Gravity Override Params())
   > Set Gravity Override(Gravity Override=@n6)
   n7 = SelectVector(A="0.1,0.35,0.35", B="1,1,0.6", bPickA=OR Boolean(A=Get bIsSkyDiving(), B=Get bIsWaterJump() -> (bIsWaterJump:bool)))
   > Set Component Linear Acc Scale(Component Linear Acc Scale:Vector=@n7) -> (Output_Get:Vector)
   > Set Component Linear Vel Scale(Component Linear Vel Scale:Vector=@n7) -> (Output_Get:Vector)
   n8 = Cast To PlayerPawn_Athena(execute=@n4.then_4, Object=TryGetPawnOwner())
   > Set head_mesh(head_mesh:Skeletal Mesh Component Object Reference=GetSkeletalMeshForPartType(self=@n8, PartType=Head)) -> (Output_Get:Skeletal Mesh Component Object Reference)
   n9 = GetOwningActor()
   n10 = GetVelocity!(self:Actor Object Reference=@n9) -> (ReturnValue:Vector)
   n11 = GetActorForwardVector!(self:Actor Object Reference=@n9) -> (ReturnValue:Vector)
   Set ForwardMovementAlpha(
    execute=@n4.then_2
    ForwardMovementAlpha:double=MapRangeClamped(Value=Dot Product!(A:Vector=@n10, B:Vector=@n11) -> (ReturnValue:double), InRangeA=0, InRangeB=400)
   ) -> (Output_Get:double)
   > Set FloatConstraint_alpha(FloatConstraint_alpha:double=SelectFloat(A=0.4, B=0, bPickA=Get bIsSkydiveFloating())) -> (Output_Get:double)
   n12 = Get Skel Mesh()
   Set LeftLegForwardBendCompensationAlpha(
    execute=@n4.then_1
    LeftLegForwardBendCompensationAlpha:double=GetAlphaFromReferenceTransformDelta(SkeletalMeshComponent=@n12, BoneName=thigh_l, BaseName=pelvis, InputRange=(X=-17,Y=-103), TwistAxis="1,0,0")
   ) -> (Output_Get:double)
   > Set RightLegForwardBendCompensationAlpha(RightLegForwardBendCompensationAlpha:double=GetAlphaFromReferenceTransformDelta(SkeletalMeshComponent=@n12, BoneName=thigh_r, BaseName=pelvis, InputRange=(X=-17,Y=-103), TwistAxis="1,0,0")) -> (Output_Get:double)
   > Set RThighFrontUpAlpha(RThighFrontUpAlpha:double=float + float(
     A=GetAlphaFromReferenceTransformDelta(SkeletalMeshComponent=@n12, BoneName=thigh_r, BaseName=pelvis, InputRange=(X=0,Y=-50), TwistAxis="1,0,0")
     B=GetAlphaFromReferenceTransformDelta(SkeletalMeshComponent=@n12, BoneName=thigh_r, BaseName=pelvis, InputRange=(X=-70,Y=-110), OutputRange=(X=0,Y=-1), TwistAxis="1,0,0")
    )) -> (Output_Get:double)
   > Set LThighFrontUpAlpha(LThighFrontUpAlpha:double=float + float(
     A=GetAlphaFromReferenceTransformDelta(SkeletalMeshComponent=@n12, BoneName=thigh_l, BaseName=pelvis, InputRange=(X=0,Y=-60), TwistAxis="1,0,0")
     B=GetAlphaFromReferenceTransformDelta(SkeletalMeshComponent=@n12, BoneName=thigh_l, BaseName=pelvis, InputRange=(X=-70,Y=-110), OutputRange=(X=0,Y=-1), TwistAxis="1,0,0")
    )) -> (Output_Get:double)
  ]}
]
variables [
 Component Linear Acc Scale {type=Vector}
 Component Linear Vel Scale {type=Vector}
 FloatConstraint_alpha {type=number}
 ForwardMovementAlpha {type=number}
 Gravity Override {type=Vector}
 Gravity Override Params {type=GravityOverrideParamsStruct}
 head_mesh {type=SkeletalMeshComponent}
 LeftLegForwardBendCompensationAlpha {type=number}
 Local Noise Offset {type=Vector}
 Local Noise Transform Location {type=Vector}
 LThighFrontUpAlpha {type=number}
 Mat Instance Dynamic {type=MaterialInstanceDynamic}
 Overlay Mat Instance Dynamic {type=MaterialInstanceDynamic}
 pickaxeSwing {type=number}
 PlayerPawn {type=FortPlayerPawnAthena}
 R1 Velocity Transform Location {type=Vector}
 R1 Velocity Transform Velocity {type=Vector}
 R2 Velocity Transform Location {type=Vector}
 R2 Velocity Transform Velocity {type=Vector}
 RightLegForwardBendCompensationAlpha {type=number}
 RThighFrontUpAlpha {type=number}
 Skel Mesh {type=SkeletalMeshComponent}
 Velocity Interp Speed {type=number}
 WPO Pho Wind Intensity {type=number}
]
